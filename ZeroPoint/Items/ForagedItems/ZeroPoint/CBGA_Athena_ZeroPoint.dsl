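; Zero Point foraged-item actor Blueprint that can be knocked away from a point by a physics impulse.
; The Impulse event waits, then pushes the StaticMesh along the normalized direction from the given origin, scaled by the impulse vector and offset by its Z value.
; After a second delay it sets RepCollision to the supplied flag, while BeginPlay and the construction script only call their parents.
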(event EventBeginPlay
  (|CallParentFunction))

(event EventImpulse
  (params (DelayBeforeImpulse :type Float%20%28double-precision%29) (DelayBeforeCollision :type Float%20%28double-precision%29) (ImpulseOrigin :type Vector :default "0, 0, 0") (ImpulseAmount :type Vector :default "0, 0, 0") (SetCollisionAfterImpulse :type Boolean))
  ;[1 1 1] Impulse Actor
  (Utilities|FlowControl|Delay DelayBeforeImpulse)
  (Physics|AddImpulse (Variables|Default|GetStaticMesh) (+ (* (Math|Vector|Normalize (- (Transformation|GetActorLocation) ImpulseOrigin)) ImpulseAmount) (.z ImpulseAmount)) "None" false)
  (Utilities|FlowControl|Delay DelayBeforeCollision)
  (Variables|Gravity|SetRepCollision SetCollisionAfterImpulse))

(fn ConstructionScript
  (|CallParentFunction))
