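; Zero Point foraged-item ability that gives the player the Zero Point active effect and handles its about-to-expire cue.
; On activation it prints "Granted", validates the avatar as a FortPlayerPawn, and applies GE_ZeroPoint_ActiveEffect to the owner.
; A looping one-second timer compares the effect's remaining duration to a five-second threshold, with a DoOnce limiting the expiry GameplayCue to a single execution.
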(event Ability|EventActivateAbility
  (bind _returnvalue (Ability|GetAvatarActorfromActorInfo))
  (Development|PrintString "Granted")
  (Utilities|IsValid _returnvalue
    (:Is%20Valid
      (bind _asfort_player_pawn (Utilities|Casting|CastToFortPlayerPawn _returnvalue)
        (:then
          (Utilities|IsValid _asfort_player_pawn
            (:Is%20Valid
              (bind _returnvalue_2 (Ability|ApplyGameplayEffectToOwner "/ZeroPoint/Items/ForagedItems/ZeroPoint/GE_ZeroPoint_ActiveEffect.GE_ZeroPoint_ActiveEffect_C"))
              (Utilities|Time|SetTimerbyEvent (delegate Custom|Check%20for%20Expiry%20Effect%20Add) 1.0 true))
            (:Is%20Not%20Valid
              _)))
        (:CastFailed
          _))
      (Ability|EndAbility))
    (:Is%20Not%20Valid
      (Ability|EndAbility))))

(event Custom|Check%20for%20Expiry%20Effect%20Add
  (bind _returnvalue (Ability|GameplayEffect|GetActiveGameplayEffectRemainingDuration (stale Ability|EventActivateAbility _returnvalue_2)))
  (if (<= _returnvalue 5.0)
    _)
  (Utilities|FlowControl|DoOnce false
    (:Completed
      (Ability|ExecuteGameplayCueOnOwner (Variables|Default|GetGC_AboutToExpire)))))
